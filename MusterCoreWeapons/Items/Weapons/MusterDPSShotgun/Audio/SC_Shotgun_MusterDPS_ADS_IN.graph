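# Sound cue for the Muster DPS shotgun's aim-down-sights-in foley.
# A Random node picks one of two equally weighted wave variations and feeds a Modulator with default settings for per-play variation.
# The cue plays quieter and slightly pitched up, routed through the weapon foley sound class and borrowing the weapon reload attenuation.
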SoundCue /MusterCoreWeapons/Items/Weapons/MusterDPSShotgun/Audio/SC_Shotgun_MusterDPS_ADS_IN
root=Modulator_0
nodes [
 {id=Modulator_0, children [Random_0]}
 {id=Random_0, children [WavePlayer_1,WavePlayer_2], properties {weights [1,1]}}
 {id=WavePlayer_1, properties {soundWaveAssetPtr=/MusterCoreWeapons/Items/Weapons/MusterDPSShotgun/Audio/SoundWaves/DPSShotgun_ADS_In_01}}
 {id=WavePlayer_2, properties {soundWaveAssetPtr=/MusterCoreWeapons/Items/Weapons/MusterDPSShotgun/Audio/SoundWaves/DPSShotgun_ADS_In_02}}
]
properties {
 attenuationSettings=&Fort_Weapon_Reload_Att
 maxDistance=2250
 pitchMultiplier=1.1
 soundClassObject=&Fort_WeaponFoley
 volumeMultiplier=0.35
}
