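# Sound cue for the Muster DPS shotgun landing after being dropped, played as a loot-class sound.
# A mixer layers an immediate drop impact wave with a landed wave that is briefly delayed and quieter, then feeds a modulator for variation.
# Weapon drop attenuation spatializes the cue, and a concurrency override caps simultaneous instances by preventing new plays.
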SoundCue /MusterCoreWeapons/Items/Weapons/MusterDPSShotgun/Audio/SC_Shotgun_MusterDPS_Drop_Land
root=Modulator_0
nodes [
 {id=Delay_0, children [WavePlayer_2], properties {delayMax=0.1, delayMin=0.1}}
 {id=Mixer_0, children [WavePlayer_0,Delay_0], properties {inputVolume [1,0.5]}}
 {id=Modulator_0, children [Mixer_0]}
 {id=WavePlayer_0, properties {soundWaveAssetPtr=/MusterCoreWeapons/Items/Weapons/MusterDPSShotgun/Audio/SoundWaves/DPS_Shotgun_Drop_Impact}}
 {id=WavePlayer_2, properties {soundWaveAssetPtr=/MusterCoreWeapons/Items/Weapons/MusterDPSShotgun/Audio/SoundWaves/DPSShotgun_Landed_01}}
]
properties {
 attenuationSettings=&Fort_Weapon_Drop_Att
 bOverrideConcurrency=true
 maxDistance=3000
 soundClassObject=&Fort_Loot_Collect
 volumeMultiplier=0.4
 attenuationOverrides {falloffDistance=1000, attenuationShapeExtents {x=0}}
 concurrencyOverrides {maxCount=3, resolutionRule=PreventNew}
}
