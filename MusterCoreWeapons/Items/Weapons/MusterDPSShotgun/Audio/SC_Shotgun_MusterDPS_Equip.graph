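# Sound cue for the equip sound of the Muster DPS shotgun in MusterCoreWeapons.
# A single wave player feeds the Shotgun_MusterDPS_Equip sound wave through a modulator at the root, left at its default settings.
# It uses the weapon foley sound class and attenuation at reduced volume, overrides concurrency, and has the debug flag switched on.
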SoundCue /MusterCoreWeapons/Items/Weapons/MusterDPSShotgun/Audio/SC_Shotgun_MusterDPS_Equip
root=Modulator_0
nodes [
 {id=Modulator_0, children [WavePlayer_3]}
 {id=WavePlayer_3, properties {soundWaveAssetPtr=/MusterCoreWeapons/Items/Weapons/MusterDPSShotgun/Audio/SoundWaves/Shotgun_MusterDPS_Equip}}
]
properties {
 attenuationSettings=&AS_Fort_WeaponFoley
 bDebug=true
 bOverrideConcurrency=true
 maxDistance=2250
 soundClassObject=&Fort_WeaponFoley
 volumeMultiplier=0.5
}
