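# Sound cue for the low-ammo sound of the Muster DPS shotgun.
# A single wave player feeds the output directly, with no randomization or modulation.
# It uses weapon foley attenuation and a first-person ranged weapon fire sound class.
# A concurrency override prevents new instances from starting once the limit is reached.
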SoundCue /MusterCoreWeapons/Items/Weapons/MusterDPSShotgun/Audio/SC_Shotgun_MusterDPS_Low_Ammo
root=WavePlayer_0
nodes [{id=WavePlayer_0, properties {soundWaveAssetPtr=/MusterCoreWeapons/Items/Weapons/MusterDPSShotgun/Audio/SoundWaves/DPS_Shotgun_LowAmmo_01}}]
properties {
 attenuationSettings=&AS_Fort_WeaponFoley
 bOverrideConcurrency=true
 maxDistance=2250
 soundClassObject=&Fort_WeaponRanged_Fire_1P_PSM
 concurrencyOverrides {resolutionRule=PreventNew}
}
