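# Sound cue for the Muster DPS shotgun's out-of-ammo dry-fire sound.
# A Random node picks one of three equally weighted dry-fire waves, which feeds a Modulator that slightly varies volume at fixed pitch.
# Playback uses weapon foley attenuation and a third-person ranged fire sound class at reduced volume, with a concurrency override that stops the oldest instance beyond a small cap.
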SoundCue /MusterCoreWeapons/Items/Weapons/MusterDPSShotgun/Audio/SC_Shotgun_MusterDPS_OutofAmmo
root=Modulator_0
nodes [
 {id=Modulator_0, children [Random_2], properties {pitchMax=1, pitchMin=1, volumeMax=0.95, volumeMin=0.85}}
 {id=Random_2, children [WavePlayer_0,WavePlayer_1,WavePlayer_2], properties {weights [1,1,1]}}
 {id=WavePlayer_0, properties {soundWaveAssetPtr=/MusterCoreWeapons/Items/Weapons/MusterDPSShotgun/Audio/SoundWaves/DPS_Shotgun_DryFire_03}}
 {id=WavePlayer_1, properties {soundWaveAssetPtr=/MusterCoreWeapons/Items/Weapons/MusterDPSShotgun/Audio/SoundWaves/DPS_Shotgun_DryFire_02}}
 {id=WavePlayer_2, properties {soundWaveAssetPtr=/MusterCoreWeapons/Items/Weapons/MusterDPSShotgun/Audio/SoundWaves/DPS_Shotgun_DryFire_01}}
]
properties {
 attenuationSettings=&AS_Fort_WeaponFoley
 bOverrideConcurrency=true
 maxDistance=2250
 soundClassObject=&Fort_WeaponRanged_Fire_3P_PSM
 volumeMultiplier=0.5
 attenuationOverrides {falloffDistance=1600}
 concurrencyOverrides {maxCount=2, resolutionRule=StopOldest}
}
