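# Sound cue for the end of the Muster DPS shotgun's reload, played as weapon foley.
# A mixer layers a random pick of three first-person pump waves, at reduced volume, with a random pick of two reload-end waves.
# The mix passes through a root modulator and uses weapon foley attenuation and sound class, with a concurrency override and the debug flag on.
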SoundCue /MusterCoreWeapons/Items/Weapons/MusterDPSShotgun/Audio/SC_Shotgun_MusterDPS_Reload_End
root=Modulator_0
nodes [
 {id=Mixer_1, children [Random_1,Random_0], properties {inputVolume [0.3,1]}}
 {id=Modulator_0, children [Mixer_1]}
 {id=Random_0, children [WavePlayer_4,WavePlayer_6], properties {weights [1,1]}}
 {id=Random_1, children [WavePlayer_0,WavePlayer_1,WavePlayer_2], properties {weights [1,1,1]}}
 {id=WavePlayer_0, properties {soundWaveAssetPtr=/MusterCoreWeapons/Items/Weapons/MusterDPSShotgun/Audio/SoundWaves/DPSShotgun_1P_Pump_01}}
 {id=WavePlayer_1, properties {soundWaveAssetPtr=/MusterCoreWeapons/Items/Weapons/MusterDPSShotgun/Audio/SoundWaves/DPSShotgun_1P_Pump_02}}
 {id=WavePlayer_2, properties {soundWaveAssetPtr=/MusterCoreWeapons/Items/Weapons/MusterDPSShotgun/Audio/SoundWaves/DPSShotgun_1P_Pump_03}}
 {id=WavePlayer_4, properties {soundWaveAssetPtr=/MusterCoreWeapons/Items/Weapons/MusterDPSShotgun/Audio/SoundWaves/DPSShotgun_Reload_End_02}}
 {id=WavePlayer_6, properties {soundWaveAssetPtr=/MusterCoreWeapons/Items/Weapons/MusterDPSShotgun/Audio/SoundWaves/DPSShotgun_Reload_End_01}}
]
properties {
 attenuationSettings=&AS_Fort_WeaponFoley
 bDebug=true
 bOverrideConcurrency=true
 maxDistance=2250
 soundClassObject=&Fort_WeaponFoley
 volumeMultiplier=0.5
}
