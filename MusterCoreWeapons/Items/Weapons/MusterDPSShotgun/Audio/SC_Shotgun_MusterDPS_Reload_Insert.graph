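# Sound cue for the Muster DPS shotgun's shell-insert sound during reload, using the weapon foley sound class and attenuation.
# A Random node picks among three equally weighted cartridge wave variants.
# A continuous modulator above it maps the MagazineCount parameter to pitch, so the insert sound changes pitch with the magazine count.
# The cue overrides concurrency settings and has its debug flag enabled.
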SoundCue /MusterCoreWeapons/Items/Weapons/MusterDPSShotgun/Audio/SC_Shotgun_MusterDPS_Reload_Insert
root=ModulatorContinuous_0
nodes [
 {
  id=ModulatorContinuous_0
  children [Random_0]
  properties {pitchModulationParams {default=5, maxInput=5, minInput=1, minOutput=0.7, parameterName=MagazineCount}}
 }
 {id=Random_0, children [WavePlayer_1,WavePlayer_2,WavePlayer_3], properties {weights [1,1,1]}}
 {
  id=WavePlayer_1
  properties {soundWaveAssetPtr=/MusterCoreWeapons/Items/Weapons/MusterDPSShotgun/Audio/SoundWaves/DPSShotgun_Reload_Cartridge_01}
 }
 {
  id=WavePlayer_2
  properties {soundWaveAssetPtr=/MusterCoreWeapons/Items/Weapons/MusterDPSShotgun/Audio/SoundWaves/DPSShotgun_Reload_Cartridge_02}
 }
 {
  id=WavePlayer_3
  properties {soundWaveAssetPtr=/MusterCoreWeapons/Items/Weapons/MusterDPSShotgun/Audio/SoundWaves/DPSShotgun_Reload_Cartridge_03}
 }
]
properties {
 attenuationSettings=&AS_Fort_WeaponFoley
 bDebug=true
 bOverrideConcurrency=true
 maxDistance=2250
 soundClassObject=&Fort_WeaponFoley
 volumeMultiplier=0.9
}
