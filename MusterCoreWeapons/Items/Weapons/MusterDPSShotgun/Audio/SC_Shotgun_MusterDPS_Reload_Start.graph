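# Sound cue for the start of the Muster DPS shotgun reload.
# A single wave player feeds the root directly with one reload-start sound wave, with no randomization or modulation nodes.
# The cue plays at reduced volume and raised pitch, using weapon foley attenuation and sound class with a concurrency override.
# The debug flag is left enabled on the cue.
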SoundCue /MusterCoreWeapons/Items/Weapons/MusterDPSShotgun/Audio/SC_Shotgun_MusterDPS_Reload_Start
root=WavePlayer_1
nodes [{
  id=WavePlayer_1
  properties {soundWaveAssetPtr=/MusterCoreWeapons/Items/Weapons/MusterDPSShotgun/Audio/SoundWaves/DPSShotgun_Reload_Start_01}
 }]
properties {
 attenuationSettings=&AS_Fort_WeaponFoley
 bDebug=true
 bOverrideConcurrency=true
 maxDistance=2250
 pitchMultiplier=1.2
 soundClassObject=&Fort_WeaponFoley
 volumeMultiplier=0.5
}
